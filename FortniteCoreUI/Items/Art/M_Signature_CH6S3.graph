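# Translucent UI material that draws the CH6S3 signature texture, apparently in flat white, with a Reveal Amount parameter to animate it appearing.
# The texture is fill-scaled about its center, and its RGB MSDF becomes an anti-aliased shape mask.
# Inverted alpha serves as a reveal-order gradient, wiped with a narrow soft edge as Reveal Amount rises.
# A second smoothstep forces the whole shape visible as the parameter nears one.
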Material /FortniteCoreUI/Items/Art/M_Signature_CH6S3
graph {
 properties {
  bAutomaticallySetUsageInEditor=false
  bUsedWithStaticMesh=false
  blendMode=BLEND_Translucent
  materialDomain=MD_UI
  usedShadingModels=Default Lit
 }
 nodes [
  Constant3Vector() {constant=<rgba 1 1 1 0>}
  n1 = ScalarParameter() {defaultValue=1, parameterName=Reveal Amount}
  n2 = Add(A=@n1) {constB=-0.025}
  n3 = TextureSample(UVs=MaterialFunctionCall(
    UVs=MaterialFunctionCall() {materialFunction=&MF_UI_UVScaling}.Scaled To Fill
    Texture Scale=Constant() {r=0.85}
   ) {materialFunction=&ScaleUVsByCenter}.UVs) {samplerType=SAMPLERTYPE_LinearColor, texture=&/T_UI_CH6S3_Signature}
  n4 = OneMinus(None=@n3.A)
  Multiply(
   A=Max(A=SmoothStep(Value=@n1) {constMin=0.9}, B=SmoothStep(Min=@n1, Max=@n2, Value=@n4) {constMin=0.9})
   B=MaterialFunctionCall!(
    Threshold=Constant() {r=0.6}
    Gradient=MaterialFunctionCall!(MSDF RGB=@n3.RGB) {materialFunction=/FortUILibrary/MaterialFunctions/Shapes/Modifiers/MF_UI_MSDFtoSDF}.SDF
    Filter Width Override
    Filter Width Fudge Factor
   ) {materialFunction=/Game/UI/Materials/AAThreshold}.Result
  )
 ]
}
